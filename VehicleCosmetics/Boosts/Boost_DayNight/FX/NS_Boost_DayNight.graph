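# Niagara system for the DayNight vehicle boost cosmetic, drawing trails and sprites at the car's boost sockets, with painted color variant overrides.
# Two ribbon emitters, a narrower colored one and a wider dark-blue one, stream backward, inherit fading source movement and undulate through an age-and-loop-driven vertical sine offset.
# A sprite emitter bursts three spawn groups, partitioned across sockets, with per-group size and curve-animated material parameters on looping particle lifetimes.
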NiagaraSystem /VehicleCosmetics/Boosts/Boost_DayNight/FX/NS_Boost_DayNight
dependencies {
 usedDataInterfaces [&NiagaraDataInterfaceCurve, &NiagaraDataInterfaceSkeletalMesh]
 usedDynamicInputs []
 usedModules [
  &self:SystemSpawnScript.SystemScriptSource.SystemScriptGraph.NiagaraNodeAssignment_0.SetVariables_#1
  &SystemState
 ]
}
emitters {
 NE_DayNight_Ribbon {
  emitterUpdateScript [
   &EmitterState(Life Cycle Mode=::NewEnumerator1, Loop Duration=1, UseLoopDelay=false)
   &VC_FilteredSocketCount()
   &VC_RateQualityHandler(High Spawn Rate=60, Low Spawn Rate=60, Medium Spawn Rate=60, Use Custom Rate?=true)
   &VC_SpawnRibbon_v1(Number of Ribbons=$Emitter.PortCount, Spawn Time=1, SpawnRate=$Emitter.Spawn Rate)
  ]
  particleSpawnScript [
   &VC_PaintedVariantsOverride(
    Cobalt=<rgba 0 0.0185 1 0.1>
    Crimson=<rgba 0.4635 0 0.0126 0.1>
    DefaultColor=<rgba 0.3763 0.0976 0.6514 0.2>
    ForestGreen=<rgba 0 0.5 0.0253 0.1>
    Grey=<rgba 0.2031 0.2031 0.2031 0.5>
    Lime=<rgba 0.2943 0.5885 0 0.1>
    Orange=<rgba 0.6724 0.2346 0 0.1>
    Pink=<rgba 0.5052 0.0605 0.3435 0.1>
    Purple=<rgba 0.2789 0 0.6795 0.1>
    R1 Cobalt:LinearColor=<rgba 0 0.0185 1 0.1>
    R1 Crimson:LinearColor=<rgba 0.4635 0 0.0126 0.1>
    R1 DefaultColor=<rgba 0.1651 0.2542 0.9131 0.2>
    R1 ForestGreen:LinearColor=<rgba 0 0.5 0.0253 0.1>
    R1 Lime:LinearColor=<rgba 0.2943 0.5885 0 0.1>
    R1 Orange:LinearColor=<rgba 0.6724 0.2346 0 0.1>
    R1 Pink:LinearColor=<rgba 0.5052 0.0605 0.3435 0.1>
    R1 Purple:LinearColor=<rgba 0.2789 0 0.6795 0.1>
    R1 Saffron:LinearColor=<rgba 0.9 0.69 0 0.1>
    R1 SkyBlue:LinearColor=<rgba 0 0.357 0.5365 0.1>
    R1 TitaniumWhite:LinearColor=<rgba 0.75 0.75 0.75 0.1>
    R3 Black:LinearColor=<rgba 0.1719 0.1719 0.1719 1>
    R3 BurntSienna:LinearColor=<rgba 0.3177 0.1227 0.0556 1>
    R3 Cobalt:LinearColor=<rgba 0.0165 0.0249 0.474 0.1>
    R3 Crimson:LinearColor=<rgba 1 0.0365 0.0626 0.1>
    R3 DefaultColor:LinearColor=<rgba 0.9131 0.1746 0.1301 0.2>
    R3 ForestGreen:LinearColor=<rgba 0 0.4427 0.0224 0.1>
    R3 Grey:LinearColor=<rgba 0.4271 0.4271 0.4271 1>
    R3 Lime:LinearColor=<rgba 0.5453 0.8698 0.0076 0.1>
    R3 Orange:LinearColor=<rgba 0.7917 0.3402 0 0.1>
    R3 Pink:LinearColor=<rgba 1 0.0781 0.7743 0.1>
    R3 Purple:LinearColor=<rgba 0.3242 0.0078 0.7396 0.1>
    R3 Saffron:LinearColor=<rgba 0.8542 0.6632 0.0356 0.1>
    R3 SkyBlue:LinearColor=<rgba 0.0485 0.5326 0.776 0.1>
    R3 TitaniumWhite:LinearColor=<rgba 2 2 2 0.1>
    Saffron=<rgba 0.9 0.69 0 0.1>
    SkyBlue=<rgba 0 0.357 0.5365 0.1>
    TitaniumWhite=<rgba 0.75 0.75 0.75 0.1>
    Use R1 Painted Overrides=true
    Use R3 Painted Overrides=true
   )
   &VC_InitializeRibbon_v1(
    Color=$Particles.Color
    Custom Offset=true
    Enable Skeletal Position=true
    Lifetime=&MergeHelper_FloatSwitch(0=1, 1=1, 2=1, 3=?, Index=$User.GameModeID, Num=::NewEnumerator1)
    Local Vector Offset=&AddVectorToPosition(Position=<-35 0 0>, Vector=$User.DTCarOffset)
    Offset Amount=1
    Ribbon Width=30
   )
   &AddVelocity(Velocity=<-1000 0 0>) {module=AddVelocity001}
  ]
  particleUpdateScript [
   &ParticleState()
   &InheritSourceMovement(
    Applied Position Scale=&VectorFromFloat(Value=&FloatFromCurve(FloatCurve=DataInterface() {propertyValues {Curve {keys [
          {interpMode=RCIM_Cubic, tangentMode=RCTM_Break, time=0, value=1}
          {arriveTangent=-2.9025, interpMode=RCIM_Cubic, tangentMode=RCTM_Break, time=1, value=0}
         ]}}}))
    ApplyForce=false
    ApplyPosition=true
    ApplyPositionFalloff=false
   )
   /SetVariables(Particles.Position=&AddVectorToPosition(
     Position=$Particles.Position
     Vector=&TransformVector(
      Vector=&MakeVector(
       X=0
       Y=0
       Z=&Sine(
        Normalized Angle=&Add_Float(
         A=&Multiply_Float(A=$Particles.NormalizedAge, B=-1.3)
         B=&Multiply_Float(A=$Emitter.LoopedAge, B=1)
        )
        Period=1
        Scale=&FloatFromCurve(FloatCurve=DataInterface() {propertyValues {Curve {keys [<at 0.1 0>,<at 0.5 1>]}}}, Scale Curve=0.6)
        Bias=0
       )
      )
      Destination Space=::Simulation
      Apply Scale=true
     )
    )) {isSetParameters=true, module=SetVariables_#2}
   &SolveForcesAndVelocity(Clamp Velocity=false, Limit Acceleration=false)
   &ScaleRibbonWidth(Ribbon Width Scale=&FloatFromCurve(
     FloatCurve=DataInterface() {propertyValues {Curve {
        defaultValue=1
        keys [
         {arriveTangent=1.4082, interpMode=RCIM_Cubic, leaveTangent=1.4082, tangentMode=RCTM_User, time=0.0143, value=0.6179}
         {
          arriveTangent=0.6389
          arriveTangentWeight=0.0852
          interpMode=RCIM_Cubic
          leaveTangent=0.6389
          leaveTangentWeight=0.0852
          tangentMode=RCTM_User
          tangentWeightMode=RCTWM_WeightedBoth
          time=0.1314
          value=0.7101
         }
         {arriveTangent=-0.1523, interpMode=RCIM_Cubic, leaveTangent=-0.1523, tangentMode=RCTM_User, time=0.3925, value=0.6916}
         {arriveTangent=-2.4411, interpMode=RCIM_Cubic, leaveTangent=-2.4411, tangentMode=RCTM_User, time=0.6182, value=0.498}
         {arriveTangent=-0.7985, interpMode=RCIM_Cubic, leaveTangent=-0.7985, tangentMode=RCTM_User, time=0.75, value=0}
        ]
       }}}
     Scale Curve=0.7
    ))
   &VC_UpdateRibbon_v1(Enable Skeletal Position=true)
  ]
  rendererClasses [&NiagaraRibbonRendererProperties]
  renderers [
   {rendererClass=&NiagaraRibbonRendererProperties}
   {rendererClass=&NiagaraRibbonRendererProperties, rendererIndex=1}
   {rendererClass=&NiagaraRibbonRendererProperties, rendererIndex=2}
  ]
 }
 NE_DayNight_Ribbon001 {
  emitterUpdateScript [
   &EmitterState(Life Cycle Mode=::NewEnumerator1, Loop Duration=1, UseLoopDelay=false)
   &VC_FilteredSocketCount()
   &VC_RateQualityHandler(High Spawn Rate=60, Low Spawn Rate=60, Medium Spawn Rate=60, Use Custom Rate?=true)
   &VC_SpawnRibbon_v1(Number of Ribbons=$Emitter.PortCount, Spawn Time=1, SpawnRate=$Emitter.Spawn Rate)
  ]
  particleSpawnScript [
   &VC_PaintedVariantsOverride(
    DefaultColor=<rgba 0.008 0.0003 0.1779 1>
    R1 Crimson:LinearColor=<rgba 0.4635 0 0.0126 1>
    R1 DefaultColor=<rgba 0.008 0.0003 0.1779 1>
    R3 Black:LinearColor=<rgba 0.2522 0.2522 0.2522 1>
    R3 BurntSienna:LinearColor=<rgba 0.7344 0.3673 0.241 1>
    R3 Cobalt:LinearColor=<rgba 0.0522 0.079 1.5 1>
    R3 Crimson:LinearColor=<rgba 1 0.25 0.2703 1>
    R3 DefaultColor:LinearColor=<rgba 3 2.008 1.14 1>
    R3 ForestGreen:LinearColor=<rgba 0 1.5 0.0758 1>
    R3 Grey:LinearColor=<rgba 0.5208 0.5208 0.5208 1>
    R3 Lime:LinearColor=<rgba 0.9404 1.5 0.0131 1>
    R3 Orange:LinearColor=<rgba 1.5 0.6446 0 1>
    R3 Pink:LinearColor=<rgba 1.5 0.4453 1.2418 1>
    R3 Purple:LinearColor=<rgba 0.5982 0.0144 1.3646 1>
    R3 Saffron:LinearColor=<rgba 1.5 1.1646 0.0625 1>
    R3 SkyBlue:LinearColor=<rgba 0.1 1.0981 1.6 1>
    R3 TitaniumWhite:LinearColor=<rgba 2 2 2 1>
    Use R1 Painted Overrides=true
    Use R3 Painted Overrides=true
   )
   &VC_InitializeRibbon_v1(
    Color=<rgba 0.0056 0.0002 0.125 1>
    Custom Offset=true
    Enable Skeletal Position=true
    Lifetime=&MergeHelper_FloatSwitch(0=1, 1=1, 2=1, 3=?, Index=$User.GameModeID, Num=::NewEnumerator1)
    Local Vector Offset=&AddVectorToPosition(Position=<-35 0 0>, Vector=$User.DTCarOffset)
    Offset Amount=1
    Ribbon Width=80
   )
   &AddVelocity(Velocity=<-1000 0 0>) {module=AddVelocity001}
  ]
  particleUpdateScript [
   &ParticleState()
   &InheritSourceMovement(
    Applied Position Scale=&VectorFromFloat(Value=&FloatFromCurve(FloatCurve=DataInterface() {propertyValues {Curve {keys [
          {interpMode=RCIM_Cubic, tangentMode=RCTM_Break, time=0, value=1}
          {arriveTangent=-2.9025, interpMode=RCIM_Cubic, tangentMode=RCTM_Break, time=1, value=0}
         ]}}}))
    ApplyForce=false
    ApplyPosition=true
    ApplyPositionFalloff=false
   )
   /SetVariables() {isSetParameters=true}
   /SetVariables(Particles.Position=&AddVectorToPosition(
     Position=$Particles.Position
     Vector=&TransformVector(
      Vector=&MakeVector(
       X=0
       Y=0
       Z=&Sine(
        Normalized Angle=&Add_Float(
         A=&Multiply_Float(A=$Particles.NormalizedAge, B=-1.3)
         B=&Multiply_Float(A=$Emitter.LoopedAge, B=1)
        )
        Period=1
        Scale=&FloatFromCurve(FloatCurve=DataInterface() {propertyValues {Curve {keys [<at 0.1 0>,<at 0.5 1>]}}}, Scale Curve=0.6)
        Bias=0
       )
      )
      Destination Space=::Simulation
      Apply Scale=true
     )
    )) {isSetParameters=true, module=SetVariables_#2}
   &SolveForcesAndVelocity(Clamp Velocity=false, Limit Acceleration=false)
   &ScaleRibbonWidth(Ribbon Width Scale=&FloatFromCurve(FloatCurve=DataInterface() {propertyValues {Curve {
        defaultValue=1
        keys [
         {arriveTangent=0.2885, interpMode=RCIM_Cubic, leaveTangent=0.2885, tangentMode=RCTM_User, time=0, value=0.5853}
         {
          arriveTangent=0.2102
          arriveTangentWeight=0.0728
          interpMode=RCIM_Cubic
          leaveTangent=0.2102
          leaveTangentWeight=0.0728
          tangentMode=RCTM_User
          tangentWeightMode=RCTWM_WeightedBoth
          time=0.1314
          value=0.5953
         }
         {arriveTangent=-0.1523, interpMode=RCIM_Cubic, leaveTangent=-0.1523, tangentMode=RCTM_User, time=0.3925, value=0.6916}
         {arriveTangent=-2.4411, interpMode=RCIM_Cubic, leaveTangent=-2.4411, tangentMode=RCTM_User, time=0.6182, value=0.498}
         {arriveTangent=-0.7985, interpMode=RCIM_Cubic, leaveTangent=-0.7985, tangentMode=RCTM_User, time=0.75, value=0}
        ]
       }}}))
   &VC_UpdateRibbon_v1(Enable Skeletal Position=true)
  ]
  rendererClasses [&NiagaraRibbonRendererProperties]
  renderers [
   {rendererClass=&NiagaraRibbonRendererProperties}
   {rendererClass=&NiagaraRibbonRendererProperties, rendererIndex=1}
   {rendererClass=&NiagaraRibbonRendererProperties, rendererIndex=2}
  ]
 }
 NE_VC_Boost_Sprite_Parent00 {
  emitterUpdateScript [
   &EmitterState(
    Inactive Response=::NewEnumerator1
    Life Cycle Mode=::NewEnumerator1
    Loop Behavior=::NewEnumerator1
    Loop Duration Mode=::NewEnumerator1
    UseLoopDelay=false
   )
   &SpawnBurst_Instantaneous(Spawn Count=2, Use Spawn Probability=false, UseLoopCountLimit=false)
   &SpawnBurst_Instantaneous(Spawn Count=2, Spawn Group=1, Use Spawn Probability=false, UseLoopCountLimit=false) {module=SpawnBurst_Instantaneous001}
   &SpawnBurst_Instantaneous(Spawn Count=2, Spawn Group=2, Use Spawn Probability=false, UseLoopCountLimit=false) {module=SpawnBurst_Instantaneous002}
   &VC_FilteredSocketCount()
   &SpawnRate(SpawnRate=50, Use Spawn Probability=false) {enabled=false}
  ]
  particleSpawnScript [
   &VC_PaintedVariantsOverride(
    R1 Black:LinearColor=<rgba 0.4348 0.4348 0.4348 1>
    R1 BurntSienna:LinearColor=<rgba 0.7656 0.2815 0.1515 1>
    R1 Cobalt:LinearColor=<rgba 0.2188 0.2702 3 1>
    R1 Crimson:LinearColor=<rgba 3 0.1719 0.2485 1>
    R1 DefaultColor=<rgba 1 0.7991 0.3712 1>
    R1 ForestGreen:LinearColor=<rgba 0.1406 3 0.2851 1>
    R1 Grey:LinearColor=<rgba 2 2 2 1>
    R1 Lime:LinearColor=<rgba 1.5547 3 0.1094 1>
    R1 Orange:LinearColor=<rgba 3 1.0973 0.0781 1>
    R1 Pink:LinearColor=<rgba 3 0.4375 2.0682 1>
    R1 Purple:LinearColor=<rgba 1.3234 0.1562 3 1>
    R1 Saffron:LinearColor=<rgba 3 2.3438 0.1875 1>
    R1 SkyBlue:LinearColor=<rgba 0.2344 2.0746 3 1>
    R1 TitaniumWhite:LinearColor=<rgba 3 3 3 1>
    Use R1 Painted Overrides=true
   )
   /SetVariables(Particles.VisTag:NiagaraInt32=&SpawnGroup()) {isSetParameters=true}
   &InitializeParticle(
    Color=<rgba 1 0.7991 0.3712 1>
    Lifetime=$Emitter.CurrentLoopDuration
    Sprite Size Mode=::NewEnumerator3
    Uniform Sprite Size=/VehicleCosmetics/FX/Niagara/Modules/MergeEmitters/VC_FloatBySpawnGroup(
     SpawnGroup or ExecIndex:NiagaraInt32=0
     CustomInt:NiagaraInt32=0
     MaximumSpawnGroups:NiagaraInt32=3
     Float SG 0:NiagaraFloat=280
     Float SG 1:NiagaraFloat=30
     Float SG 2:NiagaraFloat=30
     Float SG 3:NiagaraFloat=0
     Float SG 4:NiagaraFloat=0
     Float SG 5:NiagaraFloat=0
    )
    UsePositionOffset=false
    Write Material Random=true
   )
   &PartitionParticles(Number Of Partitions=$Emitter.PortCount)
   &SkeletalMeshLocation(
    Bone / Socket Index=$Particles.PartitionParticles.Partition
    Bone / Socket Sampling Mode=::NewEnumerator5
    MaskSpawnGroup=false
    Mesh Sampling Type=::NewEnumerator6
    Sampled Position Offset=&AddVectorToPosition(Position=<-50 0 0>, Vector=$User.DTCarOffset)
    Skeletal Mesh=$System.SkeletalMesh
   )
   &AddVelocity(Velocity=<-1500 0 0>) {enabled=false}
  ]
  particleUpdateScript [
   &ParticleState(Kill Particles When Lifetime Has Elapsed=false, Loop Particles Lifetime=true)
   &DynamicMaterialParameters(
    Index 0 Param 1=&FloatFromCurve(FloatCurve=DataInterface() {propertyValues {Curve {keys [<at 0 5 RCIM_Cubic>,<at 1 0 RCIM_Cubic>]}}})
    Index 0 Param 2=45
    Index 0 Param 3=-0.2
    Index 0 Param 4=&FloatFromCurve(FloatCurve=DataInterface() {propertyValues {Curve {keys [<at 0 1.4 RCIM_Cubic>,<at 1 12 RCIM_Cubic>]}}})
    Module.Index0Param0Write=true
    Module.Index0Param1Write=true
    Module.Index0Param2Write=true
    Module.Index0Param3Write=true
   )
   &ScaleSpriteSize(
    Scale Sprite Size Mode=::NewEnumerator0
    Uniform Scale Factor=&VC_FloatBySpawnGroup(
     SpawnGroup or ExecIndex:NiagaraInt32=0
     CustomInt:NiagaraInt32=0
     MaximumSpawnGroups:NiagaraInt32=3
     Float SG 0:NiagaraFloat=&FloatFromCurve(
      FloatCurve=DataInterface() {propertyValues {Curve {keys [{tangentMode=RCTM_User, time=0, value=0.2}, {tangentMode=RCTM_User, time=0.85, value=2}, <at 1 0>]}}}
      Scale Curve=3
     )
     Float SG 1:NiagaraFloat=0
     Float SG 2:NiagaraFloat=2
     Float SG 3:NiagaraFloat=0
     Float SG 4:NiagaraFloat=0
     Float SG 5:NiagaraFloat=0
    )
   )
   &SolveForcesAndVelocity(Clamp Velocity=false, Limit Acceleration=false)
  ]
  rendererClasses [&NiagaraSpriteRendererProperties]
  renderers [
   {rendererClass=&NiagaraSpriteRendererProperties}
   {rendererClass=&NiagaraSpriteRendererProperties, rendererIndex=1}
   {rendererClass=&NiagaraSpriteRendererProperties, rendererIndex=2}
  ]
 }
}
summary {
 systemName=NS_Boost_DayNight
 emitters [
  {emitterName=NE_DayNight_Ribbon001, rendererClasses [&NiagaraRibbonRendererProperties]}
  {emitterName=NE_VC_Boost_Sprite_Parent00, rendererClasses [&NiagaraSpriteRendererProperties]}
  {emitterName=NE_DayNight_Ribbon, rendererClasses [&NiagaraRibbonRendererProperties]}
 ]
 userVariables [
  User.bIsFrontendPreview {type=NiagaraBool}
  User.bIsPainted {type=NiagaraBool}
  User.ColorVariantID {type=NiagaraInt32}
  User.DTCarOffset {type=Vector3f, description="", defaultValue {struct=Vector3f, value=<0 0 0>}}
  User.GameModeID {type=NiagaraInt32}
  User.PaintedColor {type=LinearColor, description="", defaultValue {struct=LinearColor, value=<rgba 1 1 1 1>}}
  User.PaintedScalar {type=NiagaraFloat, defaultValue=1, description=""}
  User.PortCount {type=NiagaraInt32, defaultValue=2}
 ]
}
